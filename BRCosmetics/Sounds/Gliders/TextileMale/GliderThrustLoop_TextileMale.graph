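# Looping thrust sound cue for the TextileMale cosmetic glider, with glider-specific attenuation and audio property settings.
# A mixer layers a looping flight wave with a high flight wave, each driven by continuous modulators reading the AnyDirection parameter.
# The base layer rises in volume and pitch with that parameter, while the high layer's volume follows it.
# An enveloper on the high layer holds near full volume, then eases it down over several seconds.
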SoundCue /BRCosmetics/Sounds/Gliders/TextileMale/GliderThrustLoop_TextileMale
root=Mixer_0
nodes [
 {
  id=Enveloper_0
  children [WavePlayer_1]
  properties {
   pitchCurve {editorCurveData {keys [{time=0}]}}
   volumeCurve {editorCurveData {keys [<at 0 0.9994>, {arriveTangent=-0.0999, leaveTangent=-0.2456, time=3, value=0.9994}, <at 7 0.4>]}}
  }
 }
 {id=Mixer_0, children [ModulatorContinuous_0, ModulatorContinuous_1], properties {inputVolume [1,1]}}
 {
  id=ModulatorContinuous_0
  children [WavePlayer_0]
  properties {
   pitchModulationParams {default=0, maxOutput=1.2, minOutput=1, parameterName=AnyDirection}
   volumeModulationParams {default=0, maxOutput=1.1, minOutput=0.4, parameterName=AnyDirection}
  }
 }
 {
  id=ModulatorContinuous_1
  children [Enveloper_0]
  properties {pitchModulationParams {default=0, minOutput=1, parameterName=AnyDirection}, volumeModulationParams {default=0, parameterName=AnyDirection}}
 }
 {id=WavePlayer_0, properties {bLooping=true, soundWaveAssetPtr=&/Glider_TextileMale_Flight_LP}}
 {id=WavePlayer_1, properties {bLooping=true, soundWaveAssetPtr=&/Glider_TextileMale_Flight_Hi_LP}}
]
properties {
 attenuationSettings=&Fort_Player_Glider_Thrust_Athena_Att
 audioPropertiesSheet=&APS_Cosmetic_Glider_Loop
 maxDistance=12500
 soundClassObject=&Fort_Gliders_Land
 volumeMultiplier=0.9
 concurrencySet [null]
}
